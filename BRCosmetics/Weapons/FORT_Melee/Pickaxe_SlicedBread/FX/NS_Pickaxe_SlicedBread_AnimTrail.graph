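# Niagara animation trail for the SlicedBread pickaxe, drawing two ribbons from bones or sockets on the user-supplied skeletal mesh.
# Both emitters use shared FN_AnimTrail modules with short lifetimes; one is tinted, the other a grey, offset ribbon that narrows along a curve.
# System.Stage indexes each color curve, with impact-stage user parameters exposed.
# When the owner's execution state goes inactive, DoOnce and InterpolateOverTime blend in a color and alpha fade-out; the component-renderer Ref emitter is disabled.
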NiagaraSystem /BRCosmetics/Weapons/FORT_Melee/Pickaxe_SlicedBread/FX/NS_Pickaxe_SlicedBread_AnimTrail
dependencies {
 usedDataInterfaces [&NiagaraDataInterfaceColorCurve, &NiagaraDataInterfaceCurve, &NiagaraDataInterfaceArrayFloat]
 usedDynamicInputs [&SelectFloatFromArray]
 usedModules [
  &SystemState
  &self:SystemSpawnScript.SystemScriptSource.SystemScriptGraph.NiagaraNodeAssignment_2.SetVariables_#1
  &self:SystemSpawnScript.SystemScriptSource.SystemScriptGraph.NiagaraNodeAssignment_3.SetVariables_#2
  &self:SystemSpawnScript.SystemScriptSource.SystemScriptGraph.NiagaraNodeAssignment_4.SetVariables_#3
  &self:SystemSpawnScript.SystemScriptSource.SystemScriptGraph.NiagaraNodeAssignment_5.SetVariables_#4
 ]
}
emitters {
 FN_AnimTrail {
  emitterSpawnScript [&FN_RibbonIDOffsetUtility()]
  emitterUpdateScript [
   &EmitterState()
   &FN_SpawnAnimTrail()
  ]
  particleSpawnScript [
   &FN_InitializeAnimTrailBoneIndecies(First Bone or Socket Index=1, Second Bone or Socket Index=0)
   &FN_InitializeAnimTrail(Lifetime=0.125)
   &AddVelocity(Velocity=<0 -400 0>) {enabled=false}
  ]
  particleUpdateScript [
   &SolveForcesAndVelocity(Clamp Velocity=false, Limit Acceleration=false)
   &FN_UpdateAnimTrail(Locked Particle RibbonWidth=1, SkeletalMesh=$User.SkeletalMesh) {module=FN_UpdateAnimTrail001}
   &DoOnce(Trigger Condition:NiagaraBool=&self:CheckOwnerExecStateInactive())
   &InterpolateOverTime(
    Rate Of Change:NiagaraFloat=&Divide_Float(A=1, B=0.05)
    Target Float Value:NiagaraFloat=&MakeCustomFloatFromBool(Bool=$Particles.DoOnce.TriggeredOnce, True Float=1, False Float=0)
   )
   &Color(
    Color=&ColorFromCurve(
     ColorCurve=DataInterface() {propertyValues {
       AlphaCurve {keys [<at 0 0.5 RCIM_Cubic>,<at 1 0.35 RCIM_Cubic>]}
       BlueCurve {keys [<at 0 0.0727 RCIM_Cubic>,<at 1 0.2182>]}
       GreenCurve {keys [<at 0 0.325 RCIM_Cubic>,<at 1 0.975>]}
       RedCurve {keys [<at 0 0.5 RCIM_Cubic>,<at 1 1.5>]}
      }}
     CurveIndex=$System.Stage
    )
    Scale Alpha=&Lerp_Float(
     A=1
     B=&FloatFromCurve(
      FloatCurve=DataInterface() {propertyValues {Curve {keys [
          {
           arriveTangent=-12.607
           arriveTangentWeight=1.0031
           interpMode=RCIM_Cubic
           leaveTangent=-12.607
           leaveTangentWeight=1.0031
           tangentMode=RCTM_User
           tangentWeightMode=RCTWM_WeightedBoth
           time=0
           value=1
          }
          {
           arriveTangentWeight=0.8694
           interpMode=RCIM_Cubic
           leaveTangentWeight=0.8694
           tangentMode=RCTM_User
           tangentWeightMode=RCTWM_WeightedBoth
           time=1
           value=0
          }
         ]}}}
      CurveIndex=&ReturnNormalizedExecIndex()
     )
     Alpha=$Particles.InterpolateOverTime.InterpolateOverTime_Float.Moving Average
    )
    Scale Color=&VectorFromFloat(Value=&Lerp_Float(
      A=1
      B=&FloatFromCurve(
       FloatCurve=DataInterface() {propertyValues {Curve {keys [{tangentMode=RCTM_User, time=0, value=1}, <at 1 0 RCIM_Cubic>]}}}
       CurveIndex=&ReturnNormalizedExecIndex()
      )
      Alpha=$Particles.InterpolateOverTime.InterpolateOverTime_Float.Moving Average
     ))
   )
   &Color(Color=&LerpLinearColors(StartColor=<rgba 1 0 0 1>, EndColor=<rgba 0 1 0 1>, LerpFactor=&ReturnNormalizedExecIndex())) {enabled=false, module=Color001}
   &Color(Color=&LerpLinearColors(StartColor=<rgba 1 0 0 1>, EndColor=<rgba 0 1 0 1>, LerpFactor=$Particles.InterpolateOverTime.InterpolateOverTime_Float.Moving Average)) {enabled=false, module=Color002}
  ]
  rendererClasses [&NiagaraRibbonRendererProperties]
  renderers [{rendererClass=&NiagaraRibbonRendererProperties}]
 }
 FN_AnimTrail001 {
  emitterSpawnScript [&FN_RibbonIDOffsetUtility()]
  emitterUpdateScript [
   &EmitterState()
   &FN_SpawnAnimTrail()
  ]
  particleSpawnScript [
   &FN_InitializeAnimTrailBoneIndecies(First Bone or Socket Index=2, Second Bone or Socket Index=2)
   &FN_InitializeAnimTrail(AnimTrailWidth=10.6146, Color=<rgba 1 1 1 0.05>, Lifetime=0.125, Position Offset=<0 0.3 0.65>)
   &AddVelocity(Velocity=<0 -400 0>) {enabled=false}
  ]
  particleUpdateScript [
   &SolveForcesAndVelocity(Clamp Velocity=false, Limit Acceleration=false)
   &FN_UpdateAnimTrail(Locked Particle RibbonWidth=1, SkeletalMesh=$User.SkeletalMesh) {module=FN_UpdateAnimTrail001}
   &ScaleRibbonWidth(Ribbon Width Scale=&FloatFromCurve(FloatCurve=DataInterface() {propertyValues {Curve {keys [<at 0 1>,<at 1 0.75>]}}}))
   &DoOnce(Trigger Condition:NiagaraBool=&self:CheckOwnerExecStateInactive())
   &InterpolateOverTime(
    Rate Of Change:NiagaraFloat=&Divide_Float(A=1, B=0.05)
    Target Float Value:NiagaraFloat=&MakeCustomFloatFromBool(Bool=$Particles.DoOnce.TriggeredOnce, True Float=1, False Float=0)
   )
   &Color(
    Color=&ColorFromCurve(
     ColorCurve=DataInterface() {propertyValues {
       AlphaCurve {keys [<at 0 0.1>,<at 1 0.5>]}
       BlueCurve {keys [<at 0 0.35>,<at 1 0.35>]}
       GreenCurve {keys [<at 0 0.35>,<at 1 0.35>]}
       RedCurve {keys [<at 0 0.35>,<at 1 0.35>]}
      }}
     CurveIndex=$System.Stage
    )
    Scale Alpha=&Lerp_Float(
     A=1
     B=&FloatFromCurve(
      FloatCurve=DataInterface() {propertyValues {Curve {keys [
          {
           arriveTangent=-4.8013
           arriveTangentWeight=0.7451
           interpMode=RCIM_Cubic
           leaveTangent=-4.8013
           leaveTangentWeight=0.7451
           tangentMode=RCTM_User
           tangentWeightMode=RCTWM_WeightedBoth
           time=0
           value=1
          }
          {
           arriveTangentWeight=0.8311
           interpMode=RCIM_Cubic
           leaveTangentWeight=0.8311
           tangentMode=RCTM_User
           tangentWeightMode=RCTWM_WeightedBoth
           time=1
           value=0
          }
         ]}}}
      CurveIndex=&ReturnNormalizedExecIndex()
     )
     Alpha=$Particles.InterpolateOverTime.InterpolateOverTime_Float.Moving Average
    )
    Scale Color=&VectorFromFloat(Value=&Lerp_Float(
      A=1
      B=&FloatFromCurve(
       FloatCurve=DataInterface() {propertyValues {Curve {keys [{tangentMode=RCTM_User, time=0, value=1}, <at 1 0 RCIM_Cubic>]}}}
       CurveIndex=&ReturnNormalizedExecIndex()
      )
      Alpha=$Particles.InterpolateOverTime.InterpolateOverTime_Float.Moving Average
     ))
   )
  ]
  rendererClasses [&NiagaraRibbonRendererProperties]
  renderers [{rendererClass=&NiagaraRibbonRendererProperties}]
 }
 Ref {
  bEnabled=false
  emitterUpdateScript [
   &EmitterState(Life Cycle Mode=::NewEnumerator1, Loop Duration=1, UseLoopDelay=false)
   &SpawnBurst_Instantaneous(Use Spawn Probability=false, UseLoopCountLimit=false)
  ]
  particleSpawnScript [&InitializeParticle(Color Mode=::NewEnumerator0, UsePositionOffset=true, Write Material Random=true)]
  particleUpdateScript [&ParticleState()]
  rendererClasses [&NiagaraComponentRendererProperties]
  renderers [{rendererClass=&NiagaraComponentRendererProperties, rendererIndex=0}]
 }
}
summary {
 systemName=NS_Pickaxe_SlicedBread_AnimTrail
 emitters [
  {bEnabled=false, emitterName=Ref, rendererClasses [&NiagaraComponentRendererProperties]}
  {emitterName=FN_AnimTrail, rendererClasses [&NiagaraRibbonRendererProperties]}
  {emitterName=FN_AnimTrail001, rendererClasses [&NiagaraRibbonRendererProperties]}
 ]
 userVariables [
  User.BP_ImpactStage {type=NiagaraInt32}
  User.ImpactStage {type=NiagaraFloat, defaultValue=0, description=""}
  User.RibbonIDOffset {type=NiagaraInt32}
  User.SkeletalMesh {
   type=NiagaraDataInterfaceSkeletalMesh
   defaultValue {
    struct=NiagaraExt_VariableValue_DataInterface
    value {dataInterface=&self:NiagaraDataInterfaceSkeletalMesh_0}
   }
  }
 ]
}
