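# Idle Niagara effect for the SlicedBread pickaxe: a socket-attached glow of meshes, sprites, light and component renderer.
# One active emitter bursts a persistent particle that follows the skeletal mesh.
# System.Stage indexes curves for brightness, color and size, with system modules using array selection and interpolation over time.
# A waveform adds flicker and time-of-day scalars adjust intensity and tint; two other emitters are disabled.
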NiagaraSystem /BRCosmetics/Weapons/FORT_Melee/Pickaxe_SlicedBread/FX/NS_Pickaxe_SlicedBread_Idle
dependencies {
 usedDataInterfaces [&NiagaraDataInterfaceCurve, &NiagaraDataInterfaceColorCurve, &NiagaraDataInterfaceArrayFloat]
 usedDynamicInputs [&SelectFloatFromArray]
 usedModules [
  &SystemState
  &self:SystemSpawnScript.SystemScriptSource.SystemScriptGraph.NiagaraNodeAssignment_3.SetVariables_#1
  &InterpolateOverTime
  &self:SystemSpawnScript.SystemScriptSource.SystemScriptGraph.NiagaraNodeAssignment_0.SetVariables_#2
  &self:SystemSpawnScript.SystemScriptSource.SystemScriptGraph.NiagaraNodeAssignment_1.SetVariables_#3
  &self:SystemSpawnScript.SystemScriptSource.SystemScriptGraph.NiagaraNodeAssignment_4.SetVariables_#4
 ]
}
emitters {
 Empty {
  emitterUpdateScript [
   &EmitterState()
   &SpawnBurst_Instantaneous(Use Spawn Probability=false, UseLoopCountLimit=true)
  ]
  particleSpawnScript [
   &InitializeParticle(
    Color=<rgba 100 100 90 1>
    Mesh Scale Mode=::NewEnumerator3
    Sprite Size Mode=::NewEnumerator3
    Uniform Sprite Size=20.3513
    UsePositionOffset=false
    Write Material Random=true
   )
   &CameraOffset(Camera Offset Amount=2)
   /SetVariables(Particles.Filament Offset=0.3) {isSetParameters=true}
   /SetVariables() {isSetParameters=true}
  ]
  particleUpdateScript [
   &ParticleState(Kill Particles When Lifetime Has Elapsed=false, Let Infinitely Lived Particles Die When Emitter Deactivates=true)
   /Niagara/Noise/Waveform(
    EnableExponent1=false
    GlobalClampMax=false
    GlobalClampMin=false
    UseInput1=false
    UsePhase1=false
    "Waveform [1]":ENiagara_Waveforms=::NewEnumerator12
    "[1] Amplitude Min/Max":Vector2f=<0.65 1>
    "[1] Phase Scale"=4
   )
   &SkeletalMeshLocation(
    Bone / Socket Index=2
    MaskSpawnGroup=false
    Mesh Sampling Type=::NewEnumerator5
    Skeletal Mesh=$User.SkeletalMesh
    Socket Sampling Mode=::NewEnumerator5
   )
   /SetVariables(Particles.Velocity=&TransformVector(Vector=<0 1 0>, Destination Space=::Simulation)) {isSetParameters=true}
   &FN_Light_Attributes(
    Create Color Attribute=true
    Create Exponent Attribute=true
    Create Position Attribute=true
    Create Radius Attribute=true
    Exponent:NiagaraFloat=&FloatFromCurve(FloatCurve=DataInterface() {propertyValues {Curve {keys [<at 0 1.5>,<at 0.6666 1.5>,<at 1 2.5>]}}}, CurveIndex=$System.Stage)
    Position=&AddVectorToPosition(Position=$Particles.Initial.SkeletalMeshLocation.SampledPosition, Vector=<0 0 0>)
    Radius=&FloatFromCurve(
     FloatCurve=DataInterface() {propertyValues {Curve {keys [{tangentMode=RCTM_User, time=0, value=0}, <at 0.01 0.75>, {tangentMode=RCTM_User, time=1, value=1}]}}}
     Scale Curve=100
    )
    Scale Color=&FloatFromCurve(
     FloatCurve=DataInterface() {propertyValues {Curve {keys [{tangentMode=RCTM_User, time=0, value=0}, {tangentMode=RCTM_User, time=1, value=10}]}}}
     CurveIndex=$System.Stage
     Scale Curve=&FNTimeOfDayScalar(Night Scale=0.2, Morning Scale=2, Day Scale=2, Evening Scale=2)
    )
   )
   /SetVariables(Particles.Gen5 Inten:NiagaraFloat=&Multiply_Float(
     A=&MakeFloatFromLinearColor(LinearColor=$Particles.FN_Light_Attributes.Color, Channel:ENiagaraLinearColor_Channels=::NewEnumerator3)
     B=$Particles.Waveform.Waveform
    )) {isSetParameters=true}
   /SetVariables(
    Particles.Flare Size:Vector2f=&Lerp_Vector2(
     A=<55.5 8.5721>
     B=<60 10>
     Alpha=&FloatFromCurve(
      FloatCurve=DataInterface() {propertyValues {Curve {keys [{tangentMode=RCTM_User, time=0, value=0}, {tangentMode=RCTM_User, time=1, value=1}]}}}
      CurveIndex=$System.Stage
     )
    )
    Particles.Flare color:LinearColor=&MultiplyLinearColor(
     Linear Color=&HueShiftLinearColor(
      Hue=0
      Saturation=&FNTimeOfDayScalar(Night Scale=4, Morning Scale=1, Evening Scale=1)
      Value=1
      Linear Color=&ColorFromCurve(
       ColorCurve=DataInterface() {propertyValues {
         AlphaCurve {keys [<at 0 0.15>,<at 1 0.75>]}
         BlueCurve {keys [<at 0 0.8 RCIM_Cubic>,<at 1 0.8 RCIM_Cubic>]}
         GreenCurve {keys [<at 0 0.9541 RCIM_Cubic>,<at 1 0.9541 RCIM_Cubic>]}
         RedCurve {keys [<at 0 1 RCIM_Cubic>,<at 1 1 RCIM_Cubic>]}
        }}
       CurveIndex=$System.Stage
      )
     )
     Scale Factor=&MakeVector4(X=1, Y=1, Z=1, W=&Multiply_Float(A=0.3161, B=$Particles.Waveform.Waveform))
    )
   ) {isSetParameters=true}
   /SetVariables(
    Particles.Filament color:LinearColor=&MultiplyLinearColor(
     Linear Color=<rgba 100 93.1094 70 1>
     Scale Factor=&MakeVector4_FromFloat(Float=&FloatFromCurve(
       FloatCurve=DataInterface() {propertyValues {Curve {keys [{tangentMode=RCTM_User, time=0, value=0.05}, {tangentMode=RCTM_User, time=1, value=1}]}}}
       CurveIndex=$System.Stage
       Scale Curve=$Particles.Waveform.Waveform
      ))
    )
    Particles.Filamint Glow:Vector2f=<0.4406 4.0285>
   ) {isSetParameters=true}
   /SetVariables(Particles.Position001:NiagaraPosition=&AddVectorToPosition(Position=$Particles.Position, Vector=<-0.28 0 0>)) {isSetParameters=true}
   &ScaleColor(
    Scale RGB=&VectorFromFloat(Value=&FloatFromCurve(
      FloatCurve=DataInterface() {propertyValues {Curve {keys [{tangentMode=RCTM_User, time=0, value=0}, {tangentMode=RCTM_User, time=1, value=1}]}}}
      CurveIndex=$System.Stage
     ))
    ScaleA=true
    ScaleRGB=true
   )
   /SetVariables(
    Particles.Round Cam Offset=10
    Particles.Round Dynamic:Vector4f=<4 1 1 1>
    Particles.Round Glow Color:LinearColor=&HueShiftLinearColor(
     Hue=&FNTimeOfDayScalar(Night Scale=-0.08, Morning Scale=0, Day Scale=0, Evening Scale=0)
     Saturation=&FNTimeOfDayScalar(Night Scale=10, Morning Scale=1, Evening Scale=1)
     Value=1
     Linear Color=&MultiplyLinearColor(
      Linear Color=&ColorFromCurve(
       ColorCurve=DataInterface() {propertyValues {
         AlphaCurve {keys [<at 0 0.1>,<at 1 0.5>]}
         BlueCurve {keys [<at 0 0.5>,<at 1 0>]}
         GreenCurve {keys [<at 0 0.75>,<at 1 0.2572>]}
         RedCurve {keys [<at 0 1>,<at 1 0.5>]}
        }}
       CurveIndex=$System.Stage
      )
      Scale Factor=&MakeVector4_FromFloat(Float=&Multiply_Float(
        A=1
        B=&Multiply_Float(
         A=$Particles.Waveform.Waveform
         B=&FNTimeOfDayScalar(Night Scale=0.8, Morning Scale=1.45, Day Scale=1.25, Evening Scale=1.45)
        )
       ))
     )
    )
   ) {isSetParameters=true}
   /SetVariables(
    Particles.Mesh Glow COlor:LinearColor=&MultiplyLinearColor(
     Linear Color=<rgba 1 1 1 1>
     Scale Factor=&MakeFloat4FromVectorAndFloat(
      "Vector (RGB)"=&VectorFromFloat(Value=1)
      "Float (Alpha)"=&FloatFromCurve(
       FloatCurve=DataInterface() {propertyValues {Curve {keys [<at 0 0 RCIM_Cubic>, {arriveTangent=3.3401, interpMode=RCIM_Cubic, leaveTangent=3.3401, tangentMode=RCTM_User, time=1, value=1}]}}}
       CurveIndex=$System.Stage
       Scale Curve=0.1
      )
     )
    )
    Particles.Mesh Glow Scale:Vector3f=&VectorFromFloat(Value=&FloatFromCurve(
      FloatCurve=DataInterface() {propertyValues {Curve {keys [<at 0 1>,<at 0.7 1.1>,<at 1 2>]}}}
      CurveIndex=$System.Stage
      Scale Curve=&Lerp_Float(A=0.75, B=1, Alpha=$Particles.Waveform.Waveform)
     ))
   ) {isSetParameters=true}
  ]
  rendererClasses [
   &NiagaraMeshRendererProperties
   &NiagaraLightRendererProperties
   &NiagaraSpriteRendererProperties
   &NiagaraComponentRendererProperties
  ]
  renderers [
   {rendererClass=&NiagaraMeshRendererProperties, rendererIndex=0}
   {rendererClass=&NiagaraLightRendererProperties, rendererIndex=1}
   {rendererClass=&NiagaraSpriteRendererProperties, rendererIndex=2}
   {rendererClass=&NiagaraComponentRendererProperties, rendererIndex=3}
   {rendererClass=&NiagaraSpriteRendererProperties, rendererIndex=4}
   {rendererClass=&NiagaraSpriteRendererProperties, rendererIndex=5}
   {rendererClass=&NiagaraMeshRendererProperties, rendererIndex=6}
  ]
 }
 Empty001 {
  bEnabled=false
  emitterUpdateScript [
   &EmitterState(Life Cycle Mode=::NewEnumerator1, Loop Behavior=::NewEnumerator1, Loop Duration=1, UseLoopDelay=false)
   &SpawnBurst_Instantaneous(Use Spawn Probability=false, UseLoopCountLimit=false)
  ]
  particleSpawnScript [
   &InitializeParticle(Color Mode=::NewEnumerator0, Lifetime=5, Sprite Size Mode=::NewEnumerator3, UsePositionOffset=false, Write Material Random=true)
   &CameraOffset(Camera Offset Amount=25)
  ]
  particleUpdateScript [
   &ParticleState(Kill Particles When Lifetime Has Elapsed=false)
   &DynamicMaterialParameters(
    Index 0 Param 1=$System.Stage
    Module.Index0Param0Write=true
    Module.Index0Param1Write=true
    Module.Index0Param2Write=true
    Module.Index0Param3Write=true
   )
  ]
  rendererClasses [&NiagaraSpriteRendererProperties]
  renderers [{rendererClass=&NiagaraSpriteRendererProperties}]
 }
 Ref {
  bEnabled=false
  emitterUpdateScript [
   &EmitterState(Life Cycle Mode=::NewEnumerator1, Loop Duration=1, UseLoopDelay=false)
   &SpawnBurst_Instantaneous(Use Spawn Probability=false, UseLoopCountLimit=false)
  ]
  particleSpawnScript [&InitializeParticle(Color Mode=::NewEnumerator0, UsePositionOffset=false, Write Material Random=true)]
  particleUpdateScript [&ParticleState()]
  rendererClasses [&NiagaraComponentRendererProperties]
  renderers [{rendererClass=&NiagaraComponentRendererProperties, rendererIndex=0}]
 }
}
summary {
 systemName=NS_Pickaxe_SlicedBread_Idle
 emitters [
  {
   emitterName=Empty
   rendererClasses [
    &NiagaraMeshRendererProperties
    &NiagaraLightRendererProperties
    &NiagaraSpriteRendererProperties
    &NiagaraComponentRendererProperties
   ]
  }
  {bEnabled=false, emitterName=Ref, rendererClasses [&NiagaraComponentRendererProperties]}
  {bEnabled=false, emitterName=Empty001, rendererClasses [&NiagaraSpriteRendererProperties]}
 ]
 userVariables [
  User.BP_ImpactStage {type=NiagaraInt32}
  User.RibbonIDOffset {type=NiagaraInt32}
  User.SkeletalMesh {
   type=NiagaraDataInterfaceSkeletalMesh
   defaultValue {
    struct=NiagaraExt_VariableValue_DataInterface
    value {dataInterface=&self:NiagaraDataInterfaceSkeletalMesh_0}
   }
  }
 ]
}
